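# Parent material for the SatinCheddar emote gadget prop, adding ice and optional snow to the 2023 character material stack.
# PreFX output feeds the ice function, and a UseSnow static switch selects whether a fresnel-tinted snow layer is blended in before PostFX.
# The snow mask combines upward-facing pre-skinned normals with Perlin noise projected in pre-skinned space, keeping it stable under animation, and is cut by an IceFXMask channel.
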Material /BRCosmetics/Animation/Props/Gadgets_SatinCheddar/Materials/M_SatinCheddar_Emote_Parent
graph {
 properties {
  bUseMaterialAttributes=true
  bUsedWithMeshDeformer=true
  bUsedWithNiagaraMeshParticles=true
  bUsedWithSkeletalMesh=true
  pixelDepthOffsetMode=PDOM_Legacy
  previewMesh=/BRCosmetics/Animation/Props/Gadgets_SatinCheddar/Meshes/SatinCheddar_Gadget
  shadingModel=MSM_FromMaterialExpression
  usedShadingModels=Default Lit | Subsurface | Clear Coat | Subsurface Profile | Cloth
 }
 nodes [
  n1 = Fresnel(Normal=Transform(None=NamedRerouteUsage())) {baseReflectFraction=0, exponent=1}
  n2 = LinearInterpolate(
   A=VectorParameter() {parameterName=SnowDarkColor, defaultValue=<rgba 0.1648 0.3186 0.3906 1>}.RGB
   Alpha=Saturate(None=Multiply(A=@n1) {constB=2})
  ) {constA=1}
  LinearInterpolate(A=NamedRerouteUsage(), B=@n2) {constAlpha=0.75}
  n3 = MaterialFunctionCall() {materialFunction=&MF_CharacterParent_2023_PreFX}
  MaterialFunctionCall(
   Material=StaticSwitchParameter(True=NamedRerouteUsage(), False=NamedRerouteUsage()) {parameterName=UseSnow}
   M=@n3.M
   Pre Skinned Local Position=@n3.M
   Distance Blend=@n3.M
   Fresnel=@n3.M
  ) {materialFunction=&MF_CharacterParent_2023_PostFX}
  n4 = PreSkinnedNormal()
  n5 = VertexInterpolator(VS=@n4)
  n6 = ConstantBiasScale(None=ComponentMask(None=ShaderStageSwitch(PixelShader=@n5.PS, VertexShader=@n4)) {b=true})
  Multiply(A=@n6, B=ComponentMask(None=NamedRerouteUsage()) {b=true})
  n7 = Add(A=LocalPosition() {localOrigin=InstancePreSkinning}.XYZ, B=VectorParameter() {parameterName=NoiseOffset, defaultValue=<rgba 150 0 0 1>}.RGB)
  n8 = VertexInterpolator(VS=@n7)
  n9 = PreSkinnedNormal()
  n10 = MaterialFunctionCall(
   TextureObject=TextureObject() {texture=&/T_Perlin_Noise_M1}
   TextureSize=Constant() {r=200}
   WorldPosition=NamedRerouteDeclaration(None=ShaderStageSwitch(PixelShader=@n8.PS, VertexShader=@n7)) {name=PreSkinnedPosition, nodeColor=<rgba 1 0 0.8203 1>}
   World Space Normal=ShaderStageSwitch(PixelShader=VertexInterpolator(VS=@n9).PS, VertexShader=@n9)
  ) {materialFunction=&WorldAlignedTexture}
  Multiply(A=@n10.Z Texture, B=@n10.Z Texture)
  Multiply(A=@n2) {constB=0.75}
  Multiply(A=@n2) {constB=0.5}
  NamedRerouteDeclaration(None=MaterialFunctionCall!(In=@n3.Result, InMask=@n3.Result, InFXMask=@n3.Result) {materialFunction=&/MF_SatinCheddar_Ice}.Result) {name=InMaterialAttributes, nodeColor=<rgba 1 0.3281 0 1>}
  n11 = SetMaterialAttributes!(
   MaterialAttributes=NamedRerouteUsage()
   Base Color=@n2
   Emissive Color=Constant()
   Normal=NamedRerouteUsage()
   Roughness=Constant() {r=0.75}
   Specular=Constant() {r=0.8}
  ) {attributeSetTypes [
    #1
    #2
    #3
    #4
    #5
   ]}
  NamedRerouteDeclaration(None=BlendMaterialAttributes(A=NamedRerouteUsage(), B=@n11, Alpha=NamedRerouteUsage())) {name=OutMaterialAttributes, nodeColor=<rgba 0.0078 0 1 1>}
  NamedRerouteDeclaration(None=@n3.Tech Art Mask) {name=InFXMask, nodeColor=<rgba 1 0 0.2344 1>}
  n12 = GetMaterialAttributes(None=NamedRerouteUsage()) {attributeGetTypes [#2, #1, #3]}
  NamedRerouteDeclaration(None=@n12.Emissive Color) {name=InEmissive, nodeColor=<rgba 1 0 0.3984 1>}
  n13 = ScalarParameter() {defaultValue=0.9, parameterName=SnowHardness}
  n14 = Add(A=@n6, B=ScalarParameter() {defaultValue=0.01, parameterName=SnowHeight})
  n15 = ScalarParameter() {defaultValue=6, parameterName=SnowTransition}
  n16 = InverseLinearInterpolate(A=@n10.Z Texture, Value=Saturate(None=Power(Base=@n14, Exp=@n15)))
  n17 = ScalarParameter() {defaultValue=0.75, parameterName=SnowInfill}
  n18 = Multiply(
   A=Multiply(A=SmoothStep(Max=@n13, Value=@n16), B=LinearInterpolate(A=@n17, Alpha=@n10.Z Texture) {constA=0.25})
   B=ScalarParameter() {defaultValue=1, parameterName=SnowIntensity}
  )
  n19 = ChannelMaskParameter(None=NamedRerouteUsage()) {parameterName=IceFXMask}
  NamedRerouteDeclaration(None=Multiply(A=Saturate(None=@n18), B=OneMinus(None=@n19.R))) {name=SnowMask, nodeColor=<rgba 0 0.1797 1 1>}
  n20 = NamedRerouteUsage()
  n21 = MaterialFunctionCall(Normal=@n20, Flatness=Constant() {r=0.75}) {materialFunction=&FlattenNormal}
  n22 = ComponentMask(None=NamedRerouteUsage()) {b=true, r=true}
  n23 = TextureSample(UVs=MaterialFunctionCall(UVs=Divide(A=@n22) {constB=50}, "Rotation Angle (0-1)"=Constant() {r=0.25}) {materialFunction=&CustomRotator}.Rotated Values) {samplerType=SAMPLERTYPE_Normal, texture=&/T_SatinCheddar_normal}
  n24 = MaterialFunctionCall(
   Normal=NamedRerouteDeclaration(None=DeriveNormalZ(InXY=@n23.RGB)) {name=SnowNormal, nodeColor=<rgba 0 1 0.5313 1>}
   Flatness=Multiply(A=ScalarParameter() {defaultValue=1, parameterName=SnowNormalIntensity}) {constB=-1}
  ) {materialFunction=&FlattenNormal}
  NamedRerouteDeclaration(None=LinearInterpolate(
    A=@n20
    B=MaterialFunctionCall(BaseNormal=@n21.Result, AdditionalNormal=@n24.Result) {materialFunction=&BlendAngleCorrectedNormals}.Result
    Alpha=NamedRerouteUsage()
   )) {name=OutNormal, nodeColor=<rgba 0.5 1 0 1>}
  NamedRerouteDeclaration(None=@n12.Base Color) {name=InBaseColor, nodeColor=<rgba 0.1953 0 1 1>}
  NamedRerouteDeclaration(None=@n12.Normal) {name=InNormal, nodeColor=<rgba 1 0.5625 0 1>}
  NamedRerouteUsage()
 ]
}
